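# Niagara trail effect for the ClinchMetal VinderTech glider, with variants for gameplay and front-end display.
# Two ribbon anim-trail emitters and two socket-attached glow sprite emitters (one with lights) follow the skeletal mesh, joined by torus-spawned speedlines and a rotation-smoothed wind mesh.
# A system timeline drives intensity and glow blinking, while SineColorSwitch blends trail and glow colors.
# Time of day scales trail, wind and speedline opacity.
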NiagaraSystem /BRCosmetics/Gadgets/Assets/VinderTech_GliderChute/Glider_ClinchMetal/FX/NS_Glider_ClinchMetal_Trail
dependencies {
 usedDataInterfaces [
  &NiagaraDataInterfacePlatformSet
  &NiagaraDataInterfaceCurve
  &NiagaraDataInterfaceArrayInt32
  &NiagaraDataInterfaceArrayFloat3
  &NiagaraDataInterfaceArrayFloat
  &NiagaraDataInterfaceSocketReader
 ]
 usedDynamicInputs [
  &Sine
  &SetBoolByFloatComparison
  &RemapRange
  &MakeIntFromFloat
  &NormalizeVector
  &Add_Vector
  &ClampFloat
 ]
 usedModules [
  &self:SystemSpawnScript.SystemScriptSource.SystemScriptGraph.NiagaraNodeAssignment_0.SetVariables_#1
  &self:SystemSpawnScript.SystemScriptSource.SystemScriptGraph.NiagaraNodeAssignment_1.SetVariables_#2
  &SystemState
  &self:SystemSpawnScript.SystemScriptSource.SystemScriptGraph.NiagaraNodeAssignment_8.SetVariables_#3
  &Timeline
  &FN_Integrator
  &self:SystemSpawnScript.SystemScriptSource.SystemScriptGraph.NiagaraNodeAssignment_4.SetVariables_#4
  &self:SystemSpawnScript.SystemScriptSource.SystemScriptGraph.NiagaraNodeAssignment_5.SetVariables_#5
 ]
}
emitters {
 FN_AnimTrail001 {
  emitterSpawnScript [&FN_RibbonIDOffsetUtility()]
  emitterUpdateScript [
   &EmitterState()
   &FN_SpawnAnimTrail(Number of Anim Trails=3)
  ]
  particleSpawnScript [
   &FN_InitializeAnimTrailBoneIndecies(Second Bone or Socket Index=0)
   &FN_InitializeAnimTrailBoneIndecies(AnimTrail Selection=1, First Bone or Socket Index=3, Second Bone or Socket Index=3) {module=FN_InitializeAnimTrailBoneIndecies001}
   &FN_InitializeAnimTrailBoneIndecies(AnimTrail Selection=2, First Bone or Socket Index=4, Second Bone or Socket Index=4) {module=FN_InitializeAnimTrailBoneIndecies002}
   &FN_InitializeAnimTrail(
    AnimTrailWidth=&SelectFloatFromArray(Direct Array Index=$Particles.SpawnGroup, Float Selection Array=DataInterface() {propertyValues {FloatData [70,30,30]}})
    Lifetime=&SelectFloatFromArray(Direct Array Index=$Particles.SpawnGroup, Float Selection Array=DataInterface() {propertyValues {FloatData [0.2,0.16,0.16]}})
   )
   &AddVelocity(
    Velocity=&MakeVector(
     X=&SelectFloatFromArray(Direct Array Index=$Particles.SpawnGroup, Float Selection Array=DataInterface() {propertyValues {FloatData [0,1,1]}})
     Y=0
     Z=0
    )
    Velocity Speed Scale=&SelectFloatFromArray(Direct Array Index=$Particles.SpawnGroup, Float Selection Array=DataInterface() {propertyValues {FloatData [0,-500,500]}})
   )
   &AddVelocity(Velocity=&FN_FrontEndVector3Modifier(Gameplay=<0 0 0>, Front End=<0 -5000 0>)) {module=AddVelocity001}
  ]
  particleUpdateScript [
   &SolveForcesAndVelocity(Clamp Velocity=false, Limit Acceleration=false)
   &FN_UpdateAnimTrail(Locked Particle Color:LinearColor=&MultiplyLinearColor(
     Linear Color=$Particles.Initial.Color
     Scale Factor=&MakeFloat4FromVectorAndFloat("Vector (RGB)"=<1 1 1>, "Float (Alpha)"=$System.Timeline.NormalizedTime)
    )) {enabled=false, module=FN_UpdateAnimTrail001}
   &FN_UpdateAnimTrail()
   &ScaleColor(
    Scale Alpha=&Multiply_Float(
     A=&Multiply_Float(
      A=&FloatFromCurve(FloatCurve=DataInterface() {propertyValues {Curve {keys [
           {arriveTangent=11.4189, interpMode=RCIM_Cubic, leaveTangent=11.4189, tangentMode=RCTM_User, time=0, value=0}
           {arriveTangent=-5.8842, interpMode=RCIM_Cubic, leaveTangent=-5.8842, tangentMode=RCTM_User, time=0.1, value=2.3}
           {interpMode=RCIM_Cubic, tangentMode=RCTM_User, time=0.8, value=0}
          ]}}})
      B=&FloatFromCurve(
       FloatCurve=DataInterface() {propertyValues {Curve {keys [
           {interpMode=RCIM_Cubic, tangentMode=RCTM_User, time=0, value=2}
           {interpMode=RCIM_Cubic, tangentMode=RCTM_User, time=0.5, value=1}
           {tangentMode=RCTM_User, time=1, value=2}
          ]}}}
       CurveIndex=Linked() {linkedVariable {name=NPC.FortniteNPC.TimeOfDay_0-1, type=NiagaraFloat}}
      )
     )
     B=&FN_FrontEndFloatModifier(Gameplay=$System.Timeline.NormalizedTime, Front End=0)
    )
    Scale RGB=&Lerp_Vector(
     A=&MakeVectorFromLinearColorRGB(Color=&LerpLinearColors(
       StartColor=<rgba 1 0.8152 0.744 1>
       EndColor=$System.ColorA
       LerpFactor=&FloatFromCurve(FloatCurve=DataInterface() {propertyValues {Curve {
           defaultValue=0
           keys [
            {arriveTangent=2.1037, interpMode=RCIM_Cubic, leaveTangent=2.1037, tangentMode=RCTM_User, time=0, value=0}
            {interpMode=RCIM_Cubic, tangentMode=RCTM_User, time=1, value=1}
           ]
          }}})
      ))
     B=&MakeVectorFromLinearColorRGB(Color=&LerpLinearColors(
       StartColor=<rgba 0.8179 0.9718 1 1>
       EndColor=$System.ColorB
       LerpFactor=&FloatFromCurve(FloatCurve=DataInterface() {propertyValues {Curve {
           defaultValue=1
           keys [
            {arriveTangent=22.3378, interpMode=RCIM_Cubic, leaveTangent=22.3378, tangentMode=RCTM_User, time=0, value=0.3}
            {interpMode=RCIM_Cubic, tangentMode=RCTM_User, time=0.1, value=1}
           ]
          }}})
      ))
     Alpha=$System.SineColorSwitch
    )
    ScaleA=true
    ScaleRGB=true
   ) {enabled=false, module=ScaleColor001}
   &ScaleColor(
    Scale Alpha=&Multiply_Float(
     A=&Multiply_Float(
      A=&FloatFromCurve(FloatCurve=DataInterface() {propertyValues {Curve {keys [
           {arriveTangent=11.4189, interpMode=RCIM_Cubic, leaveTangent=11.4189, tangentMode=RCTM_User, time=0, value=0}
           {arriveTangent=-5.8842, interpMode=RCIM_Cubic, leaveTangent=-5.8842, tangentMode=RCTM_User, time=0.1, value=2.3}
           {interpMode=RCIM_Cubic, tangentMode=RCTM_User, time=0.8, value=0}
          ]}}})
      B=&FloatFromCurve(
       FloatCurve=DataInterface() {propertyValues {Curve {keys [
           {interpMode=RCIM_Cubic, tangentMode=RCTM_User, time=0, value=2}
           {interpMode=RCIM_Cubic, tangentMode=RCTM_User, time=0.5, value=1}
           {tangentMode=RCTM_User, time=1, value=2}
          ]}}}
       CurveIndex=Linked() {linkedVariable {name=NPC.FortniteNPC.TimeOfDay_0-1, type=NiagaraFloat}}
      )
     )
     B=&FN_FrontEndFloatModifier(
      Gameplay=$System.Timeline.NormalizedTime
      Front End=&FloatFromCurve(
       FloatCurve=DataInterface() {propertyValues {Curve {
          defaultValue=1
          keys [{interpMode=RCIM_Cubic, tangentMode=RCTM_User, time=1.5, value=0}, {interpMode=RCIM_Cubic, tangentMode=RCTM_User, time=3, value=1}]
         }}}
       CurveIndex=$System.Age
      )
     )
    )
    Scale RGB=&Lerp_Vector(
     A=&MakeVectorFromLinearColorRGB(Color=&LerpLinearColors(
       StartColor=<rgba 1 0.8152 0.744 1>
       EndColor=$System.ColorA
       LerpFactor=&FloatFromCurve(FloatCurve=DataInterface() {propertyValues {Curve {
           defaultValue=0
           keys [
            {arriveTangent=2.1037, interpMode=RCIM_Cubic, leaveTangent=2.1037, tangentMode=RCTM_User, time=0, value=0}
            {interpMode=RCIM_Cubic, tangentMode=RCTM_User, time=1, value=1}
           ]
          }}})
      ))
     B=&MakeVectorFromLinearColorRGB(Color=&LerpLinearColors(
       StartColor=<rgba 0.8179 0.9718 1 1>
       EndColor=$System.ColorB
       LerpFactor=&FloatFromCurve(FloatCurve=DataInterface() {propertyValues {Curve {
           defaultValue=1
           keys [
            {arriveTangent=22.3378, interpMode=RCIM_Cubic, leaveTangent=22.3378, tangentMode=RCTM_User, time=0, value=0.3}
            {interpMode=RCIM_Cubic, tangentMode=RCTM_User, time=0.1, value=1}
           ]
          }}})
      ))
     Alpha=$System.SineColorSwitch
    )
    ScaleA=true
    ScaleRGB=true
   )
  ]
  rendererClasses [&NiagaraRibbonRendererProperties]
  renderers [{rendererClass=&NiagaraRibbonRendererProperties}]
 }
 Glow_Sharp {
  emitterUpdateScript [
   &EmitterState(Life Cycle Mode=::NewEnumerator1, Loop Behavior=::NewEnumerator1, Loop Duration Mode=::NewEnumerator1, UseLoopDelay=false)
   &SpawnBurst_Instantaneous(Spawn Count=5, Use Spawn Probability=false, UseLoopCountLimit=false)
  ]
  particleSpawnScript [
   &InitializeParticle(
    Color Mode=::NewEnumerator0
    Lifetime=2.5
    Sprite Size Mode=::NewEnumerator3
    Uniform Sprite Size=70
    UsePositionOffset=false
    Write Material Random=true
   )
   &PartitionParticles(Number Of Partitions=5)
  ]
  particleUpdateScript [
   &ParticleState(
    Kill Particles When Lifetime Has Elapsed=false
    Let Infinitely Lived Particles Die When Emitter Deactivates=true
    Loop Particles Lifetime=true
   )
   &SkeletalMeshLocation(
    Bone / Socket Index=&SelectIntFromArray(Direct Array Index=$Particles.PartitionParticles.Partition, Int Selection Array=DataInterface() {propertyValues {IntData [0,1,2,3,4]}})
    MaskSpawnGroup=false
    Mesh Sampling Type=::NewEnumerator5
    Sampled Position Offset=&SelectVectorFromArray(
     Direct Array Index=$Particles.PartitionParticles.Partition
     Vector Selection Array=DataInterface() {propertyValues {FloatData [<0 12 1>, <0 0 -2>, <0 -4 0>, <-1 0 0>, <1 0 3>]}}
    )
    Skeletal Mesh=$System.Skeletal Mesh
    Socket Sampling Mode=::NewEnumerator5
   )
   &CameraOffset(Camera Offset Amount=20)
   &ScaleColor(
    Scale Alpha=&Multiply_Float(
     A=&FloatFromCurve(
      FloatCurve=DataInterface() {propertyValues {Curve {
         defaultValue=1
         keys [
          {interpMode=RCIM_Cubic, tangentMode=RCTM_User, time=0, value=0}
          {arriveTangent=5.9199, interpMode=RCIM_Cubic, tangentMode=RCTM_Break, time=0.5, value=1}
         ]
        }}}
      CurveIndex=$System.Age
     )
     B=&Lerp_Float(
      A=&FloatFromCurve(
       FloatCurve=DataInterface() {propertyValues {Curve {
          defaultValue=1
          keys [
           {interpMode=RCIM_Cubic, tangentMode=RCTM_Break, time=0, value=1}
           {arriveTangent=0.0146, interpMode=RCIM_Cubic, leaveTangent=-2.5, tangentMode=RCTM_Break, time=0.1, value=1}
           {arriveTangent=-3.3333, interpMode=RCIM_Cubic, leaveTangent=0.0765, tangentMode=RCTM_Break, time=0.12, value=0.025}
           {arriveTangent=0.0723, interpMode=RCIM_Cubic, leaveTangent=-0.116, tangentMode=RCTM_Break, time=0.88, value=0}
           {arriveTangent=0.1674, interpMode=RCIM_Cubic, leaveTangent=-0.0036, tangentMode=RCTM_Break, time=0.9, value=1}
           {interpMode=RCIM_Cubic, tangentMode=RCTM_Break, time=1, value=1}
          ]
         }}}
       Scale Curve=0.7
      )
      B=&FloatFromCurve(
       FloatCurve=DataInterface() {propertyValues {Curve {
          defaultValue=1
          keys [
           {interpMode=RCIM_Cubic, tangentMode=RCTM_Break, time=0, value=1}
           {arriveTangent=0.0146, interpMode=RCIM_Cubic, leaveTangent=-2.5, tangentMode=RCTM_Break, time=0.1, value=1}
           {arriveTangent=-3.3333, interpMode=RCIM_Cubic, leaveTangent=0.0765, tangentMode=RCTM_Break, time=0.12, value=0.025}
           {arriveTangent=-0.0005, interpMode=RCIM_Cubic, leaveTangent=3.4821, tangentMode=RCTM_Break, time=0.38, value=0.025}
           {arriveTangent=4.2391, interpMode=RCIM_Cubic, leaveTangent=0.3114, tangentMode=RCTM_Break, time=0.4, value=1}
           {arriveTangent=0.2455, interpMode=RCIM_Cubic, leaveTangent=-4.3478, tangentMode=RCTM_Break, time=0.6, value=1}
           {arriveTangent=-3.7419, interpMode=RCIM_Cubic, leaveTangent=0.0944, tangentMode=RCTM_Break, time=0.62, value=0}
           {arriveTangent=0.0723, interpMode=RCIM_Cubic, leaveTangent=-0.116, tangentMode=RCTM_Break, time=0.88, value=0}
           {arriveTangent=0.1674, interpMode=RCIM_Cubic, leaveTangent=-0.0036, tangentMode=RCTM_Break, time=0.9, value=1}
           {interpMode=RCIM_Cubic, tangentMode=RCTM_Break, time=1, value=1}
          ]
         }}}
       Scale Curve=0.7
      )
      Alpha=$System.Timeline.Time
     )
    )
    Scale RGB=&MakeVectorFromLinearColorRGB(Color=&LerpLinearColors(StartColor=<rgba 1 0.6488 0.5135 1>, EndColor=$System.ColorB, LerpFactor=$System.SineColorSwitch))
    ScaleA=true
    ScaleRGB=true
   )
   &ScaleSpriteSize(
    Scale Sprite Size Mode=::NewEnumerator0
    Uniform Scale Factor=&Lerp_Float(A=1, B=1.2, Alpha=$System.Timeline.Time)
   )
  ]
  rendererClasses [&NiagaraSpriteRendererProperties]
  renderers [{rendererClass=&NiagaraSpriteRendererProperties}]
 }
 Glow_Soft {
  emitterUpdateScript [
   &EmitterState(
    Inactive Response=::NewEnumerator1
    Life Cycle Mode=::NewEnumerator1
    Loop Behavior=::NewEnumerator1
    Loop Duration Mode=::NewEnumerator1
    UseLoopDelay=false
   )
   &SpawnBurst_Instantaneous(Spawn Count=5, Use Spawn Probability=false, UseLoopCountLimit=false) {module=SpawnBurst_Instantaneous002}
  ]
  particleSpawnScript [
   &PartitionParticles(Number Of Partitions=5)
   &InitializeParticle(Color Mode=::NewEnumerator0, Sprite Size Mode=::NewEnumerator3, Uniform Sprite Size=50, UsePositionOffset=false, Write Material Random=true)
  ]
  particleUpdateScript [
   &ParticleState(
    Kill Particles When Lifetime Has Elapsed=false
    Let Infinitely Lived Particles Die When Emitter Deactivates=true
    Loop Particles Lifetime=true
   )
   &SkeletalMeshLocation(
    Bone / Socket Index=&SelectIntFromArray(Direct Array Index=$Particles.PartitionParticles.Partition, Int Selection Array=DataInterface() {propertyValues {IntData [0,1,2,3,4]}})
    MaskSpawnGroup=false
    Mesh Sampling Type=::NewEnumerator5
    Sampled Position Offset=&SelectVectorFromArray(
     Direct Array Index=$Particles.PartitionParticles.Partition
     Vector Selection Array=DataInterface() {propertyValues {FloatData [<0 0 0>, <0 0 3>, <0 0 0>, <2 0 -2>, <-2 0 -2>]}}
    )
    Skeletal Mesh=$System.Skeletal Mesh
    Socket Sampling Mode=::NewEnumerator5
   )
   &CameraOffset(Camera Offset Amount=55)
   &ScaleColor(
    Scale Alpha=&FloatFromCurve(
     FloatCurve=DataInterface() {propertyValues {Curve {
        defaultValue=1
        keys [
         {interpMode=RCIM_Cubic, tangentMode=RCTM_User, time=0, value=0}
         {arriveTangent=5.9199, interpMode=RCIM_Cubic, tangentMode=RCTM_Break, time=0.5, value=1}
        ]
       }}}
     CurveIndex=$System.Age
     Scale Curve=1.5
    )
    Scale RGB=&MakeVectorFromLinearColorRGB(Color=&LerpLinearColors(StartColor=<rgba 1 0.6488 0.5135 1>, EndColor=$System.ColorB, LerpFactor=$System.SineColorSwitch))
    ScaleA=true
    ScaleRGB=true
   ) {module=ScaleColor001}
   /SetVariables(Particles.ColorLight:LinearColor=&Multiply_LinearColorByFloat(
     Linear Color=&LerpLinearColors(StartColor=<rgba 1 0.6488 0.5135 1>, EndColor=$System.ColorB, LerpFactor=$System.SineColorSwitch)
     Float=&FloatFromCurve(
      FloatCurve=DataInterface() {propertyValues {Curve {
         defaultValue=1
         keys [
          {interpMode=RCIM_Cubic, tangentMode=RCTM_User, time=0, value=0}
          {arriveTangent=5.9199, interpMode=RCIM_Cubic, tangentMode=RCTM_Break, time=0.5, value=1}
         ]
        }}}
      CurveIndex=$System.Age
      Scale Curve=5
     )
    )) {isSetParameters=true}
   &ScaleSpriteSize(
    Scale Sprite Size Mode=::NewEnumerator0
    Uniform Scale Factor=&Lerp_Float(A=1, B=1.3, Alpha=$System.Timeline.Time)
   )
  ]
  rendererClasses [&NiagaraSpriteRendererProperties, &NiagaraLightRendererProperties]
  renderers [
   {rendererClass=&NiagaraSpriteRendererProperties}
   {rendererClass=&NiagaraLightRendererProperties, rendererIndex=1}
  ]
 }
 Speedlines {
  emitterUpdateScript [
   &EmitterState(
    Delay First Loop Only=true
    Inactive Response=::NewEnumerator1
    Life Cycle Mode=::NewEnumerator1
    Loop Delay=&FN_FrontEndFloatModifier(Gameplay=3.5, Front End=3)
    Loop Duration=3
    UseLoopDelay=true
   )
   &SpawnRate(
    SpawnRate=&FN_FrontEndFloatModifier(
     Gameplay=&Multiply_Float(
      A=&Multiply_Float(A=200, B=1)
      B=&Lerp_Float(A=-0.8, B=1, Alpha=$System.Timeline.NormalizedTime)
     )
     Front End=90
    )
    Use Spawn Probability=false
   )
  ]
  particleSpawnScript [
   &InitializeParticle(
    Lifetime Max=0.12
    Lifetime Min=0.1
    Lifetime Mode=::NewEnumerator1
    Position Offset=&FN_FrontEndVector3Modifier(Gameplay=<0 0 0>, Front End=<-30 35 50>)
    Sprite Size Max=<2 7>
    Sprite Size Min=<1.75 2>
    Sprite Size Mode=::NewEnumerator2
    UsePositionOffset=true
    Write Material Random=true
   )
   &TorusLocation(
    Align Torus With Axis=true
    Handle Radius=0
    Large Radius:NiagaraFloat=&RandomRangeFloat(
     Minimum=&FN_FrontEndFloatModifier(Gameplay=300, Front End=150)
     Maximum=&FN_FrontEndFloatModifier(Gameplay=420, Front End=220)
    )
    MaskSpawnGroup=false
    Module.Override Randomness=false
    Offset=&FN_FrontEndVector3Modifier(Gameplay=<0 100 0>, Front End=<0 -35 0>)
    Torus Axis=&FN_FrontEndVector3Modifier(Gameplay=<0 1 0>, Front End=<0.5 1 0>)
   )
   &AddVelocityFromPoint(
    Default Position:NiagaraPosition=$Engine.Emitter.SimulationPosition
    Module.Use Falloff Distance=false
    Origin Offset=&AddVectorToPosition(Position=<0 500 0>, Vector=&Lerp_Vector(A=<0 0 0>, B=<0 500 0>, Alpha=$User.RightDot))
    Velocity Strength=&FN_FrontEndFloatModifier(Gameplay=1500, Front End=200)
   ) {module=AddVelocityFromPoint001}
   &AddVelocity(
    Velocity=&FN_FrontEndVector3Modifier(Gameplay=<0 500 0>, Front End=<-300 -500 0>)
    Velocity Speed Scale=&FN_FrontEndFloatModifier(Gameplay=1, Front End=0)
   )
  ]
  particleUpdateScript [
   &SolveForcesAndVelocity(Clamp Velocity=false, Limit Acceleration=false)
   &ParticleState()
   &ScaleSpriteSizeBySpeed(Max Scale Factor=<1 50>, Min Scale Factor=<0.01 0.05>, Velocity Threshold=1)
   &ScaleColor(
    Scale Alpha=&Multiply_Float(
     A=&FloatFromCurve(
      FloatCurve=DataInterface() {propertyValues {Curve {keys [
          {interpMode=RCIM_Cubic, tangentMode=RCTM_User, time=0, value=0}
          {interpMode=RCIM_Cubic, tangentMode=RCTM_User, time=0.5, value=1}
          {tangentMode=RCTM_User, time=1, value=0}
         ]}}}
      Scale Curve=&FNTimeOfDayScalar(Night Scale=0.6, Morning Scale=0.8, Evening Scale=0.8)
     )
     B=&Multiply_Float(
      A=&RandomRangeFloat(Minimum=0.3, Maximum=0.35)
      B=&FN_FrontEndFloatModifier(Gameplay=1, Front End=0.3)
     )
    )
    ScaleA=true
    ScaleRGB=true
   )
  ]
  rendererClasses [&NiagaraSpriteRendererProperties]
  renderers [{rendererClass=&NiagaraSpriteRendererProperties}]
 }
 Trail_Sides {
  emitterSpawnScript [&FN_RibbonIDOffsetUtility()]
  emitterUpdateScript [
   &EmitterState(
    Inactive Response=::NewEnumerator1
    Life Cycle Mode=::NewEnumerator1
    Loop Behavior=::NewEnumerator1
    Loop Duration Mode=::NewEnumerator1
    UseLoopDelay=false
   )
   &FN_SpawnAnimTrail(Number of Anim Trails=4)
  ]
  particleSpawnScript [
   &FN_InitializeAnimTrailBoneIndecies(First Bone or Socket Index=5, Second Bone or Socket Index=5)
   &FN_InitializeAnimTrailBoneIndecies(AnimTrail Selection=1, First Bone or Socket Index=6, Second Bone or Socket Index=6) {module=FN_InitializeAnimTrailBoneIndecies001}
   &FN_InitializeAnimTrailBoneIndecies(AnimTrail Selection=2, First Bone or Socket Index=7, Second Bone or Socket Index=7) {module=FN_InitializeAnimTrailBoneIndecies002}
   &FN_InitializeAnimTrailBoneIndecies(AnimTrail Selection=3, First Bone or Socket Index=8, Second Bone or Socket Index=8) {module=FN_InitializeAnimTrailBoneIndecies003}
   &FN_InitializeAnimTrail(AnimTrailWidth=20, Lifetime=0.2, SkeletalMesh=$System.Skeletal Mesh)
   &AddVelocity(Velocity=&FN_FrontEndVector3Modifier(Gameplay=<0 -900 0>, Front End=<0 -1500 0>))
  ]
  particleUpdateScript [
   &SolveForcesAndVelocity(Clamp Velocity=false, Limit Acceleration=false)
   &FN_UpdateAnimTrail(
    Locked Particle Color:LinearColor=&MultiplyLinearColor(
     Linear Color=$Particles.Initial.Color
     Scale Factor=&MakeFloat4FromVectorAndFloat("Vector (RGB)"=<1 1 1>, "Float (Alpha)"=$System.Timeline.NormalizedTime)
    )
    Locked Particle RibbonWidth=1
    SkeletalMesh=$User.SkeletalMesh
   )
   /SetVariables(Particles.Position=&AddVectorToPosition(Position=$Particles.Position, Vector=&TransformVector(Vector=<0 0 0>, Destination Space=::Simulation))) {isSetParameters=true}
   /SetVariables(Particles.RibbonU0Override=$Particles.NormalizedAge) {isSetParameters=true}
   &ScaleColor(
    Scale Alpha=&Multiply_Float(
     A=&Multiply_Float(
      A=&FloatFromCurve(FloatCurve=DataInterface() {propertyValues {Curve {
          defaultValue=0
          keys [
           {interpMode=RCIM_Cubic, tangentMode=RCTM_User, time=0, value=0}
           {arriveTangent=2.8812, interpMode=RCIM_Cubic, leaveTangent=2.8812, tangentMode=RCTM_User, time=1, value=1}
          ]
         }}})
      B=&FloatFromCurve(
       FloatCurve=DataInterface() {propertyValues {Curve {keys [
           {interpMode=RCIM_Cubic, tangentMode=RCTM_User, time=0, value=2}
           {interpMode=RCIM_Cubic, tangentMode=RCTM_User, time=0.5, value=1}
           {tangentMode=RCTM_User, time=1, value=2}
          ]}}}
       CurveIndex=Linked() {linkedVariable {name=NPC.FortniteNPC.TimeOfDay_0-1, type=NiagaraFloat}}
      )
     )
     B=&FN_FrontEndFloatModifier(
      Gameplay=&FloatFromCurve(
       FloatCurve=DataInterface() {propertyValues {Curve {keys [{interpMode=RCIM_Cubic, tangentMode=RCTM_User, time=0, value=0}, {interpMode=RCIM_Cubic, tangentMode=RCTM_User, time=1, value=1}]}}}
       CurveIndex=$System.Timeline.NormalizedTime
       Scale Curve=0.5
      )
      Front End=&FloatFromCurve(
       FloatCurve=DataInterface() {propertyValues {Curve {
          defaultValue=1
          keys [{interpMode=RCIM_Cubic, tangentMode=RCTM_User, time=1.5, value=0}, {interpMode=RCIM_Cubic, tangentMode=RCTM_User, time=3, value=1}]
         }}}
       CurveIndex=$System.Age
      )
     )
    )
    Scale RGB=&Lerp_Vector(
     A=&MakeVectorFromLinearColorRGB(Color=&LerpLinearColors(
       StartColor=<rgba 1 0.8152 0.744 1>
       EndColor=$System.ColorA
       LerpFactor=&FloatFromCurve(FloatCurve=DataInterface() {propertyValues {Curve {
           defaultValue=0
           keys [
            {arriveTangent=2.1037, interpMode=RCIM_Cubic, leaveTangent=2.1037, tangentMode=RCTM_User, time=0, value=0}
            {interpMode=RCIM_Cubic, tangentMode=RCTM_User, time=1, value=1}
           ]
          }}})
      ))
     B=&MakeVectorFromLinearColorRGB(Color=&LerpLinearColors(
       StartColor=<rgba 0.8179 0.9718 1 1>
       EndColor=$System.ColorB
       LerpFactor=&FloatFromCurve(FloatCurve=DataInterface() {propertyValues {Curve {
           defaultValue=1
           keys [
            {arriveTangent=22.3378, interpMode=RCIM_Cubic, leaveTangent=22.3378, tangentMode=RCTM_User, time=0, value=0.3}
            {interpMode=RCIM_Cubic, tangentMode=RCTM_User, time=0.1, value=1}
           ]
          }}})
      ))
     Alpha=$System.SineColorSwitch
    )
    ScaleA=true
    ScaleRGB=true
   ) {module=ScaleColor002}
  ]
  rendererClasses [&NiagaraRibbonRendererProperties]
  renderers [{rendererClass=&NiagaraRibbonRendererProperties}]
 }
 Wind {
  emitterUpdateScript [
   &EmitterState(
    Inactive Response=::NewEnumerator1
    Life Cycle Mode=::NewEnumerator1
    Loop Behavior=::NewEnumerator1
    Loop Duration Mode=::NewEnumerator1
    UseLoopDelay=false
   )
   &SpawnBurst_Instantaneous(Use Spawn Probability=false, UseLoopCountLimit=true)
  ]
  particleSpawnScript [
   &InitializeParticle(
    Color=&MakeLinearColorFromVectorAndFloat("Vector (RGB)"=<1 1 1>, "Float (Alpha)"=&Multiply_Float(A=1, B=1))
    Lifetime=0.5
    Mesh Scale=&Multiply_Vector(
     A=&VectorFromFloat(Value=&Multiply_Float(A=&RandomRangeFloat(Minimum=2.5, Maximum=2.5), B=1))
     B=<0.7 0.7 0.9>
    )
    Mesh Scale Mode=::NewEnumerator4
    Sprite Size Mode=::NewEnumerator3
    Uniform Sprite Size=999
    UsePositionOffset=true
    Write Material Random=true
   )
   &InitialMeshOrientation(
    Mesh Axis To Orient=<0 1 0>
    Mesh Orientation Mode=::NewEnumerator2
    Orientation Vector=<0 1 1>
    Rotation=&FN_FrontEndVector3Modifier(Gameplay=<0.45 0 0>, Front End=<0 0 0>)
    Use Rotation Vector=true
   )
  ]
  particleUpdateScript [
   &ParticleState(Kill Particles When Lifetime Has Elapsed=false)
   &SkeletalMeshLocation(
    Bone Sampling Mode=::NewEnumerator4
    MaskSpawnGroup=false
    Orientation Sampling=::NewEnumerator1
    Sampled Position Offset=&FN_FrontEndVector3Modifier(Gameplay=<0 0 -50>, Front End=<0 0 50>)
    Sampled Scale Coordinate Space:ENiagaraCoordinateSpace=::Simulation
    Skeletal Mesh=$System.Skeletal Mesh
   ) {module=SkeletalMeshLocation001}
   &InterpolateOverTime(Rate Of Change=5, Target Quat Value:Quat4f=$StackContext.SkeletalMeshLocation001.SampledRotation, Variable Type=::NewEnumerator5)
   /SetVariables(Particles.MeshOrientation=&MultiplyQuaternion(Quaternion A=$StackContext.InterpolateOverTime.InterpolateOverTime_Quaternion.Moving Average, Quaternion B=$StackContext.InitialRotation)) {isSetParameters=true}
   &ScaleColor(
    Scale Alpha=&Multiply_Float(
     A=&Multiply_Float(
      A=&FN_PickFloatDependOnPlatform(TargetPlatformActive=0.03, TargetPlatformNoActive=0.002)
      B=&FloatFromCurve(
       FloatCurve=DataInterface() {propertyValues {Curve {
          defaultValue=1
          keys [
           {interpMode=RCIM_Cubic, tangentMode=RCTM_User, time=0, value=1}
           {interpMode=RCIM_Cubic, tangentMode=RCTM_User, time=0.5, value=1.5}
           {tangentMode=RCTM_User, time=1, value=1}
          ]
         }}}
       CurveIndex=Linked() {linkedVariable {name=NPC.FortniteNPC.TimeOfDay_0-1, type=NiagaraFloat}}
      )
     )
     B=&FN_FrontEndFloatModifier(Gameplay=$System.Timeline.NormalizedTime, Front End=0.2)
    )
    ScaleA=true
    ScaleRGB=false
   )
  ]
  rendererClasses [&NiagaraMeshRendererProperties]
  renderers [{rendererClass=&NiagaraMeshRendererProperties, rendererIndex=0}]
 }
}
summary {
 systemName=NS_Glider_ClinchMetal_Trail
 emitters [
  {emitterName=Wind, rendererClasses [&NiagaraMeshRendererProperties]}
  {emitterName=Glow_Sharp, rendererClasses [&NiagaraSpriteRendererProperties]}
  {emitterName=Glow_Soft, rendererClasses [&NiagaraSpriteRendererProperties, &NiagaraLightRendererProperties]}
  {emitterName=Trail_Sides, rendererClasses [&NiagaraRibbonRendererProperties]}
  {emitterName=Speedlines, rendererClasses [&NiagaraSpriteRendererProperties]}
  {emitterName=FN_AnimTrail001, rendererClasses [&NiagaraRibbonRendererProperties]}
 ]
 userVariables [
  User.bIsFrontEnd {type=NiagaraBool}
  User.bIsFrontEndPreview {type=NiagaraBool}
  User.bIsFullyDeployed {type=NiagaraBool}
  User.ForwardDot {type=NiagaraFloat, defaultValue=0, description=""}
  User.RibbonIDOffset {type=NiagaraInt32}
  User.RightDot {type=NiagaraFloat, defaultValue=0, description=""}
  User.SkeletalMesh {
   type=NiagaraDataInterfaceSkeletalMesh
   defaultValue {
    struct=NiagaraExt_VariableValue_DataInterface
    value {dataInterface=&self:NiagaraDataInterfaceSkeletalMesh_0}
   }
  }
 ]
}
