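# Character-parent master material for the ClinchMetal glider cosmetic, adding an animated two-color emissive glow.
# It wraps custom logic between the MF_CharacterParent_2023 PreFX and PostFX functions, passing material attributes through named reroutes.
# The emissive is fully desaturated into a grayscale mask, then tinted by a blend of EmissiveColor_A and EmissiveColor_B.
# A time-driven sine, remapped and shaped by an SCurve, drives the color cycle with exposed speed and exponent.
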Material /BRCosmetics/Gadgets/Assets/VinderTech_GliderChute/Glider_ClinchMetal/Materials/M_Glider_ClinchMetal
graph {
 properties {
  bUseMaterialAttributes=true
  bUsedWithMeshDeformer=true
  bUsedWithMorphTargets=true
  bUsedWithSkeletalMesh=true
  pixelDepthOffsetMode=PDOM_Legacy
  shadingModel=MSM_FromMaterialExpression
  usedShadingModels=Default Lit | Subsurface | Clear Coat | Subsurface Profile | Cloth
 }
 nodes [
  n1 = MaterialFunctionCall() {materialFunction=&MF_CharacterParent_2023_PreFX}
  MaterialFunctionCall(Material=NamedRerouteUsage(), M=@n1.M, Pre Skinned Local Position=@n1.M, Distance Blend=@n1.M, Fresnel=@n1.M) {materialFunction=&MF_CharacterParent_2023_PostFX}
  NamedRerouteDeclaration(None=@n1.Result) {name=IN_Material, nodeColor=<rgba 1 0.9141 0 1>}
  n2 = Desaturation(
   None=GetMaterialAttributes(None=NamedRerouteUsage()) {attributeGetTypes [#1]}.Emissive Color
   Fraction=Constant() {r=1}
  )
  n3 = VectorParameter() {group=ClinchMetal, parameterName=EmissiveColor_A, sortPriority=0, defaultValue=<rgba 1.5 0.4 0.06 1>}
  n4 = Sine(None=Multiply(A=Time(), B=ScalarParameter() {defaultValue=1, group=ClinchMetal, parameterName=EmissiveCycleSpeed, sortPriority=2})) {period=3}
  n5 = LinearInterpolate(
   A=@n3.RGB
   B=VectorParameter() {group=ClinchMetal, parameterName=EmissiveColor_B, sortPriority=1, defaultValue=<rgba 0 0.58 2 1>}.RGB
   Alpha=MaterialFunctionCall(
    In=ConstantBiasScale(None=@n4)
    Power=ScalarParameter() {defaultValue=2, group=ClinchMetal, parameterName=EmissiveCycleExponent, sortPriority=3}
   ) {materialFunction=&SCurve}.Result
  )
  NamedRerouteDeclaration(None=SetMaterialAttributes(MaterialAttributes=NamedRerouteUsage(), Emissive Color=Multiply(A=@n2, B=@n5) {constA=1}) {attributeSetTypes [#1]}) {name=OUT_Material, nodeColor=<rgba 0.5938 0 1 1>}
 ]
}
